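; Blueprint that displays fake shadow proxies as instances on one instanced static mesh component.
; The construction script calls UpdateInstances, which clears the component and adds a world-space instance for each entry in the InstanceData FakeShadowData struct.
; Show and hide shadow proxy events, mirrored by enable and disable fake shadows custom events, toggle the component's visibility without propagating to children.
; A Clear custom event removes all instances.
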(event EventHideShadowProxy
  (Rendering|SetVisibility (Variables|Default|GetInstancedStaticMesh) false false))

(event Custom|disablefakeshadows
  (Rendering|SetVisibility (Variables|Default|GetInstancedStaticMesh) false false))

(event EventShowShadowProxy
  (Rendering|SetVisibility (Variables|Default|GetInstancedStaticMesh) true false))

(event Custom|enablefakeshadows
  (Rendering|SetVisibility (Variables|Default|GetInstancedStaticMesh) true false))

(event Custom|Clear%20
  (Components|InstancedStaticMesh|ClearInstances (Variables|Default|GetInstancedStaticMesh)))

(fn UpdateInstances
  (Components|InstancedStaticMesh|ClearInstances (Variables|Default|GetInstancedStaticMesh))
  (for _array_element (Utilities|Struct|BreakFakeShadowData (Variables|Default|GetInstanceData))
    (Components|InstancedStaticMesh|AddInstance (Variables|Default|GetInstancedStaticMesh) _array_element true)))

(fn ConstructionScript
  (CallFunction|UpdateInstances))
