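# Unlit alpha-composite material for the Shadow Proxy Generator that repacks a frame from the RT texture parameter.
# It samples the red channel of the texture's four quadrants, packing three into an RGB vector and using the fourth for a separate inverted scalar.
# Both results are multiplied by an edge mask, the minimum of parameter-driven per-axis border vignettes and a centered sphere mask, which fades the frame toward its borders.
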Material /Game/ContentCreationTools/ShadowProxyGenerator/Materials/M_FrameCopy
graph {
 properties {blendMode=BLEND_AlphaComposite, pixelDepthOffsetMode=PDOM_Legacy, shadingModel=MSM_Unlit, usedShadingModels=Unlit}
 nodes [
  n1 = TextureCoordinate() {uTiling=0.5, vTiling=0.5}
  n2 = Add(A=@n1, B=Constant2Vector() {r=0.5})
  n3 = Add(A=@n1, B=Constant2Vector() {g=0.5})
  n4 = TextureCoordinate() {uTiling=0.5, vTiling=0.5}
  n5 = ComponentMask(None=@n4) {r=true}
  n6 = ScalarParameter() {defaultValue=0.2, parameterName=BorderVignette}
  n7 = Subtract(A=Abs(None=Subtract(A=@n5) {constB=0.25}), B=@n6)
  n8 = Subtract(A=Constant() {r=0.25}, B=@n6)
  n9 = ScalarParameter() {defaultValue=0.1, parameterName=BorderVignetteOffset}
  n10 = Add(A=Saturate(None=Divide(A=@n7, B=@n8)), B=@n9)
  n11 = ComponentMask(None=@n4) {g=true}
  n12 = Subtract(A=Abs(None=Subtract(A=@n11) {constB=0.25}), B=@n6)
  n13 = Add(A=@n9, B=Saturate(None=Divide(A=@n12, B=@n8)))
  n14 = Min(
   A=Min(A=OneMinus(None=@n10), B=OneMinus(None=@n13))
   B=SphereMask(A=@n4, B=Constant2Vector() {g=0.25, r=0.25}) {attenuationRadius=0.275, hardnessPercent=60}
  )
  Multiply(
   A=@n14
   B=MaterialFunctionCall(
    X=TextureSampleParameter2D(UVs=@n1) {parameterName=RT, texture=/Engine/EngineResources/Black_Low.Black}.R
    Y=TextureSampleParameter2D(UVs=@n2) {parameterName=RT, texture=&Black}.R
    Z=TextureSampleParameter2D(UVs=@n3) {parameterName=RT, texture=&Black}.R
   ) {materialFunction=&MakeFloat3}.Result
  )
  n15 = Add(A=@n1, B=Constant2Vector() {g=0.5, r=0.5})
  OneMinus(None=Multiply(A=@n14, B=TextureSampleParameter2D(UVs=@n15) {parameterName=RT, texture=&Black}.R))
 ]
}
