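# Translucent unlit alpha-composite material in the Shadow Proxy Generator tool, apparently previewing a blended proxy shadow.
# A scalar drives four-component tent weights, dotted with RT texture channels to crossfade between stored shadow states.
# The result gets a soft threshold from ww and EdgeShift, then Opacity and DepthFade set the blend toward ShadowColor.
# Position math combines a UV2-scaled displacement along axis "ax" with a clamped CameraOffset.
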Material /Game/ContentCreationTools/ShadowProxyGenerator/Materials/M_PreviewBlend
graph {
 properties {blendMode=BLEND_AlphaComposite, opacityMaskClipValue=0.5, pixelDepthOffsetMode=PDOM_Legacy, shadingModel=MSM_Unlit, usedShadingModels=Unlit}
 nodes [
  n1 = Multiply(A=ComponentMask(None=ScalarParameter()) {r=true}) {constB=3}
  n2 = Add(A=@n1, B=Constant4Vector() {constant=<rgba 1 0 -1 -2>}.RGBA)
  n3 = OneMinus(None=Add(A=@n1, B=Constant4Vector() {constant=<rgba 0 -1 -2 -3>}.RGBA))
  n4 = TextureSampleParameter2D() {parameterName=RT, texture=/Engine/EngineResources/Black_Low.Black}
  n5 = DotProduct(A=Saturate(None=Min(A=@n2, B=@n3)), B=AppendVector(A=@n4.RGB, B=@n4.RGB))
  Abs(None=@n5)
  n6 = WorldPosition() {worldPositionShaderOffset=WPT_ExcludeAllShaderOffsets}
  n7 = Multiply(
   A=ComponentMask(None=TextureCoordinate() {coordinateIndex=2}) {g=true}
   B=Normalize(VectorInput=VectorParameter() {parameterName=ax, defaultValue=<rgba 0 1 0 0>}.RGB)
  )
  Add(
   A=MaterialFunctionCall(Offset Amount=ScalarParameter() {defaultValue=768, parameterName=Offset}, World Position=Add(A=@n6.XYZ, B=@n7)) {materialFunction=/Engine/Functions/Engine_MaterialFunctions02/WorldPositionOffset/CameraOffset}.Clamped Camera Offset
   B=@n7
  )
  n8 = ScalarParameter() {defaultValue=0.1, parameterName=ww}
  n9 = Subtract(B=Add(A=@n8, B=ScalarParameter() {parameterName=EdgeShift})) {constA=0.5, constB=0.5}
  n10 = Saturate(None=Divide(A=Subtract(A=@n5, B=@n9) {constB=0}, B=@n8))
  n11 = ScalarParameter() {defaultValue=0.92, parameterName=Opacity}
  LinearInterpolate(
   B=VectorParameter() {parameterName=ShadowColor, defaultValue=<rgba 0 0 0.01 0>}.RGB
   Alpha=DepthFade(Opacity=Multiply(A=@n10, B=@n11), FadeDistance=ScalarParameter() {defaultValue=500, parameterName=FD})
  )
  MaterialFunctionCall!(X range Y range=Constant4Vector() {constant=<rgba 0 1 0 1>}.RGBA, Gradient Height=Constant(), ShowGrid?, Thickness, Color, "f(x)") {materialFunction=/Engine/Functions/Engine_MaterialFunctions01/Debug/PlotFunctionOnGraph}
  OneMinus(None=@n10)
  ScalarParameter() {parameterName=T}
  TextureCoordinate()
 ]
}
