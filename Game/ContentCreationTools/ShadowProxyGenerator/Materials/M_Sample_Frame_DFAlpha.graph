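# Unlit additive material in the shadow proxy generator that processes a captured render target (RT) with two custom HLSL searches.
# For texels with nonzero alpha, an eight-direction search copies RGB from the nearest near-zero-alpha texel, blended by MPC_ImposterCapture's Dilation value.
# A brute-force ring search turns the Mask-selected channel into a normalized signed distance field, then inverts it.
# MipRT is declared but unconnected.
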Material /Game/ContentCreationTools/ShadowProxyGenerator/Materials/M_Sample_Frame_DFAlpha
graph {
 properties {blendMode=BLEND_Additive, pixelDepthOffsetMode=PDOM_Legacy, shadingModel=MSM_Unlit, usedShadingModels=Unlit}
 nodes [
  MaterialExpressionTextureObjectParameter_1 = TextureObjectParameter() {parameterName=RT, texture=&Black}
  MaterialExpressionScalarParameter_12 = ScalarParameter() {defaultValue=256, parameterName=Res}
  n1 = Multiply(A=TextureCoordinate(), B=@MaterialExpressionScalarParameter_12) {constB=1024}
  MaterialExpressionAdd_7 = Add(
   A=Divide(A=Floor(None=@n1), B=@MaterialExpressionScalarParameter_12) {constB=1024}
   B=Divide(B=@MaterialExpressionScalarParameter_12) {constA=0.5, constB=2048}
  )
  MaterialExpressionScalarParameter_13 = ScalarParameter() {defaultValue=64, parameterName=ms}
  n2 = Custom!(
   Tex=@MaterialExpressionTextureObjectParameter_1
   UV=@MaterialExpressionAdd_7
   Resolution=@MaterialExpressionScalarParameter_12
   MaxSteps=@MaterialExpressionScalarParameter_13
   TextureSize=@MaterialExpressionScalarParameter_12
  ) {
   code="//////////////// UV Dilation ///////////////////////////\r\n\r\n\r\n//** Tex        **// Input Texture Object storing Volume Data\r\n//** UV         **// Input float2 for UVs\r\n//** Resolution **// Resolution of render target\r\n\r\n\r\nfloat texelsize = 1 / Resolution;\r\nfloat mindist = 10000000;\r\n\r\nfloat2 offsets[8] = {float2(-1,0), float2(1,0), float2(0,1), float2(0,-1), float2(-1,1), float2(1,1), float2(1,-1), float2(-1,-1)};\r\n\r\nfloat4 sample = Tex.SampleLevel(TexSampler, UV, 0);\r\n\r\nfloat4 curminsample = sample;\r\n\r\n\r\n\r\nif(sample.a > 0)\r\n{\r\n\r\n\tint j = 0;\r\n\twhile(j < MaxSteps)\r\n\t{\r\n\t\t\r\n\t\tj++;\r\n\r\n\t\tint i = 0;\r\n\r\n\t\twhile (i < 8)\r\n\t\t{\r\n\r\n\t\t\tfloat2 curUV = UV + offsets[i] * texelsize * j;\r\n\t\t\tfloat4 offsetsample = Tex.SampleLevel(TexSampler, curUV, 0);\r\n\r\n\t\t\tif(offsetsample.a <= 0.01)\r\n\t\t\t{\r\n\r\n\t\t\t\tfloat curdist = length(UV - curUV);\r\n\r\n\t\t\t\tif (curdist < mindist)\r\n\t\t\t\t{\r\n\t\t\t\t\tfloat2 projectUV = curUV + offsets[i] * texelsize * j * 1;\r\n\t\t\t\t\tfloat3 direction = Tex.SampleLevel(TexSampler, projectUV, 0);\r\n\t\t\t\t\t\r\n\t\t\t\t\tif(direction.x != 0 || direction.y != 0 || direction.z != 0)\r\n\t\t\t\t\t{\r\n\t\t\t\t\t\r\n\t\t\t\t\t\tmindist = curdist;\r\n\t\t\t\t\t\tcurminsample = offsetsample;\r\n\r\n\t\t\t\t\t}\r\n\t\t\t\t\telse\r\n\t\t\t\t\t{\r\n\t\t\t\t\tmindist = curdist;\r\n\t\t\t\t\tcurminsample = offsetsample;\r\n\t\t\t\t\t}\r\n\t\t\t\t}\r\n\t\t\t}\r\n\r\n\t\r\n\t\t\ti++;\r\n\t\t}\r\n\r\n\t}\r\n}\r\n\t\r\nreturn curminsample.xyz;"
   inputs [
    {
     inputName=Tex
     input {expression=&self:MaterialExpressionTextureObjectParameter_1}
    }
    {inputName=UV, input {expression=&self:MaterialExpressionAdd_7}}
    {
     inputName=Resolution
     input {expression=&self:MaterialExpressionScalarParameter_12}
    }
    {
     inputName=MaxSteps
     input {expression=&self:MaterialExpressionScalarParameter_13}
    }
    {
     inputName=TextureSize
     input {expression=&self:MaterialExpressionScalarParameter_12}
    }
   ]
  }
  LinearInterpolate(
   A=TextureSampleParameter2D() {parameterName=RT, texture=&Black}.RGB
   B=@n2
   Alpha=CollectionParameter() {collection=/Game/ImpostorBaker/MPC/MPC_ImposterCapture, parameterName=Dilation}
  )
  MaterialExpressionTextureObjectParameter_2 = TextureObjectParameter() {parameterName=RT, texture=&Black}
  MaterialExpressionScalarParameter_11 = ScalarParameter() {defaultValue=256, parameterName=TextureSize}
  MaterialExpressionMin_4 = Min(A=ScalarParameter() {defaultValue=64, parameterName=MaxR}) {constB=64}
  MaterialExpressionTextureCoordinate_4 = TextureCoordinate()
  n3 = VectorParameter() {parameterName=Mask, defaultValue=<rgba 0 0 0 1>}
  MaterialExpressionAppendVector_5 = AppendVector(A=@n3.RGB, B=@n3.RGB)
  n4 = Custom!(
   Tex=@MaterialExpressionTextureObjectParameter_2
   TextureSize=@MaterialExpressionScalarParameter_11
   MaxRadius=@MaterialExpressionMin_4
   UV=@MaterialExpressionTextureCoordinate_4
   ChannelMask=@MaterialExpressionAppendVector_5
  ) {
   code="\r\n\r\nfloat2 curoffset = 0;\r\nfloat curdist = 0;\r\nfloat mindist = MaxRadius;\r\nfloat texelsize = 1 / TextureSize;\r\nint laststep = 0;\r\nint newmax = MaxRadius;\r\n\r\nUV = ( floor(UV * TextureSize) / TextureSize ) + ( 0.5 / TextureSize);\r\n\r\n\r\nfloat startsample = dot(ChannelMask, Tex.SampleLevel(TexSampler, UV, 0));\r\nint startinside = startsample;\r\n\r\n//i is the current radius level\r\nint i = 0;\r\nwhile( i < min(MaxRadius, newmax) )\r\n{\r\ni++;\r\n\t//8 neighbors per texel radius\r\n\tfloat totneighbors = i * 8;\r\n\t//j is the current neighbor index\r\n\tint j = 0;\r\n\twhile ( j < totneighbors )\r\n\t{\r\n\t\t\r\n\t\tfloat progress = j / totneighbors;\r\n\t\tcuroffset.x=saturate(4*progress)-saturate(4*(progress-0.5));\r\n\t\tcuroffset.y=saturate(4*(progress-0.25))-saturate(4*(progress-0.75));\r\n\r\n\t\t//normalize and offset by radius\r\n\t\tcuroffset = (curoffset * totneighbors / 4) - i;\r\n\r\n\t\tfloat2 curUV = UV + curoffset * texelsize;\r\n\t\tcurUV = ( floor(curUV * TextureSize) / TextureSize ) + ( 0.5 / TextureSize);\r\n\t\tcurUV = max( curUV, 0.5 / TextureSize );\r\n\t\tcurUV = min( curUV, 1 - (0.5 / TextureSize) );\r\n\r\n\r\n\t\tfloat cursample = dot(ChannelMask, Tex.SampleLevel(TexSampler, curUV, 0));\r\n\r\n\t\t//int circlemask = ( curoffset.x*curoffset.x + curoffset.y*curoffset.y );\r\n\r\n\t\tif( cursample != startsample )\r\n\t\t{\r\n\t\t\tcurdist = length(curoffset);\r\n\t\t\tmindist = min( mindist, curdist );\r\n\t\t}\r\n\t\t\r\n\tj++;\r\n\r\n\t}\r\n\r\n\tif ( mindist < MaxRadius &&&& newmax == MaxRadius)\r\n\t{\r\n\t\tnewmax = ceil( i * sqrt(2.0) );\t\r\n\t}\r\n\r\n\r\n}\r\n\r\nfloat output = (mindist - 0.5) / (MaxRadius - 0.5);\r\n\r\nif (startinside == 0.0)\r\n{\r\n\toutput *= -1;\r\n}\r\n\r\noutput =  (output + 1) * 0.5;\r\nreturn output;\r\n\r\n"
   outputType=CMOT_Float1
   inputs [
    {
     inputName=Tex
     input {expression=&self:MaterialExpressionTextureObjectParameter_2}
    }
    {
     inputName=TextureSize
     input {expression=&self:MaterialExpressionScalarParameter_11}
    }
    {
     inputName=MaxRadius
     input {expression=&self:MaterialExpressionMin_4}
    }
    {
     inputName=UV
     input {expression=&self:MaterialExpressionTextureCoordinate_4}
    }
    {
     inputName=ChannelMask
     input {expression=&self:MaterialExpressionAppendVector_5}
    }
   ]
  }
  OneMinus(None=@n4)
  TextureObjectParameter() {parameterName=MipRT, texture=&Black}
 ]
}
