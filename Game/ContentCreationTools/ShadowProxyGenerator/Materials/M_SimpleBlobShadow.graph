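# M_SimpleBlobShadow is an unlit, alpha-composite translucent material for a soft, dark blob shadow in the shadow proxy generator tools.
# An exponential radial gradient, softened by depth fade, drives a HeightLerp that a world-XY-projected dirt normal map's red channel breaks up.
# The resulting alpha feeds a shadow color blend and an opacity multiplier.
# A CameraOffset node supplies a parameterized world position offset.
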Material /Game/ContentCreationTools/ShadowProxyGenerator/Materials/M_SimpleBlobShadow
graph {
 properties {
  blendMode=BLEND_AlphaComposite
  pixelDepthOffsetMode=PDOM_Legacy
  shadingModel=MSM_Unlit
  translucencyPass=MTP_BeforeDOF
  usedShadingModels=Unlit
 }
 nodes [
  n1 = TextureSample(UVs=MaterialFunctionCall!(Scale=ScalarParameter() {defaultValue=5000, parameterName=zie}) {materialFunction=/Engine/Functions/Engine_MaterialFunctions01/Coordinates/WorldPosition-XY}.Result) {samplerType=SAMPLERTYPE_Normal, texture=/Game/Packages/DS_Fortnight/ExperimentalTextures/Dirt_N}
  ConstantBiasScale(None=@n1.R)
  Constant()
  n2 = ScalarParameter() {defaultValue=0.6, parameterName=R}
  n3 = ScalarParameter() {defaultValue=2, parameterName=Density}
  n4 = MaterialFunctionCall(
   Transition Phase=DepthFade(
    Opacity=MaterialFunctionCall(Radius=@n2, Density=@n3) {materialFunction=&RadialGradientExponential}.RadialGradientExponential
    FadeDistance=ScalarParameter() {defaultValue=1024, parameterName=asdf}
   )
   Height Texture=LinearInterpolate(B=@n1.R, Alpha=ScalarParameter() {defaultValue=1, parameterName=None_1}) {constA=0.5}
  ) {materialFunction=&HeightLerp}
  LinearInterpolate(B=VectorParameter() {parameterName=ShadowColor, defaultValue=<rgba 0 0 0.01 0>}.RGB, Alpha=@n4.Alpha)
  MaterialFunctionCall(Offset Amount=ScalarParameter() {defaultValue=1024, parameterName=Offset}) {materialFunction=/Engine/Functions/Engine_MaterialFunctions02/WorldPositionOffset/CameraOffset}
  Multiply(A=@n4.Alpha, B=ScalarParameter() {defaultValue=0.92, parameterName=Opacity})
  Power(
   Base=TextureSample() {texture=/Engine/EngineVolumetrics/Fogsheet/Materials/T_EV_FogSheet_Falloff_03}.R
   Exp=ScalarParameter() {defaultValue=0.3}
  )
 ]
}
